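# Sound cue for a vault silo gate door opening, mixed from three staggered, pitch-shifted layers.
# A lowered-pitch garage gate slam starts at once, then a raised-pitch silo small-door open follows, and a deeply lowered ship hatch open lands last.
# Each layer has its own attenuation, with the silo door carrying much farther using natural falloff and a distance low-pass filter.
# The cue is routed to the Fort_Enemy_Attack sound class.
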SoundCue /Game/Sounds/Fort_Doors/World/Vault/Silo_Door_Gate_Open_Cue
root=Mixer_0
nodes [
 {
  id=Attenuation_0
  children [Modulator_0]
  properties {
   bOverrideAttenuation=true
   attenuationOverrides {
    bAttenuateWithLPF=true
    dBAttenuationAtMax=-35
    distanceAlgorithm=NaturalSound
    falloffDistance=15000
    lPFFrequencyAtMax=300
    lPFFrequencyAtMin=10000
    lPFRadiusMax=10000
    nonSpatializedRadiusStart=200
    attenuationShapeExtents {x=900}
   }
  }
 }
 {
  id=Attenuation_1
  children [Modulator_1]
  properties {
   bOverrideAttenuation=true
   attenuationOverrides {distanceAlgorithm=Logarithmic, falloffDistance=4500, nonSpatializedRadiusStart=200, attenuationShapeExtents {x=700}}
  }
 }
 {
  id=Attenuation_2
  children [Delay_3]
  properties {
   bOverrideAttenuation=true
   attenuationOverrides {distanceAlgorithm=Logarithmic, falloffDistance=4500, nonSpatializedRadiusStart=200, attenuationShapeExtents {x=700}}
  }
 }
 {id=Delay_2, children [WavePlayer_3], properties {delayMax=0.54, delayMin=0.54}}
 {id=Delay_3, children [Modulator_2], properties {delayMax=2.8, delayMin=2.8}}
 {id=Mixer_0, children [Attenuation_1, Attenuation_0, Attenuation_2], properties {inputVolume [0.9,1.1,0.8]}}
 {id=Modulator_0, children [Delay_2], properties {pitchMax=1.1, pitchMin=1.1, volumeMax=1, volumeMin=1}}
 {id=Modulator_1, children [WavePlayer_2], properties {pitchMax=0.9, pitchMin=0.9, volumeMax=1, volumeMin=1}}
 {id=Modulator_2, children [WavePlayer_0], properties {pitchMax=0.7, pitchMin=0.7, volumeMax=1, volumeMin=1}}
 {id=WavePlayer_0, properties {soundWaveAssetPtr=/Game/Sounds/Fort_Doors/World/Door_Metal_Ship_Hatch_Open_01}}
 {
  id=WavePlayer_2
  properties {soundWaveAssetPtr=/Game/Sounds/Helios/Environment_Actors/NitroBossVehicleGarage/Nitro_BossVehicleGarage_Gate_Open_Slam_01}
 }
 {id=WavePlayer_3, properties {soundWaveAssetPtr=&/Silo_SmallDoor_Open}}
]
properties {
 bOverrideConcurrency=true
 maxDistance=26614.2852
 soundClassObject=/Game/Sounds/FortSoundClassesAndModes/Fort_Enemy_Attack
 volumeMultiplier=1.5
 attenuationOverrides {
  bEnableListenerFocus=true
  distanceAlgorithm=Logarithmic
  falloffDistance=1200
  focusDistanceScale=0.9
  nonFocusVolumeAttenuation=0.4
  stereoSpread=700
  attenuationShapeExtents=<250 120 -60.9851>
  customAttenuationCurve {editorCurveData {keys [{arriveTangent=-2.3812, interpMode=RCIM_Cubic, leaveTangent=-2.3812, tangentMode=RCTM_User, time=0, value=1}, <at 1 0.01 RCIM_Cubic>]}}
 }
}
