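# Sound cue for the movement of a vault door gate, reusing a Nitro boss vehicle garage gate-opening wave.
# A mixer combines that wave, through a logarithmic attenuation, with a finite loop over a longer-range, low-pass-filtered attenuation that has no child, so that branch is silent.
# The cue routes to the Fort_Enemy_Attack sound class and overrides concurrency.
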SoundCue /Game/Sounds/Fort_Doors/World/Vault/Vault_Door_Gate_Movement_Cue
root=Mixer_0
nodes [
 {
  id=Attenuation_0
  children [null]
  properties {
   bOverrideAttenuation=true
   attenuationOverrides {
    bAttenuateWithLPF=true
    dBAttenuationAtMax=-35
    distanceAlgorithm=NaturalSound
    falloffDistance=15000
    lPFFrequencyAtMax=300
    lPFFrequencyAtMin=10000
    lPFRadiusMax=10000
    nonSpatializedRadiusStart=200
    attenuationShapeExtents {x=900}
   }
  }
 }
 {
  id=Attenuation_1
  children [WavePlayer_4]
  properties {
   bOverrideAttenuation=true
   attenuationOverrides {distanceAlgorithm=Logarithmic, falloffDistance=4500, nonSpatializedRadiusStart=200, attenuationShapeExtents {x=700}}
  }
 }
 {id=Looping_0, children [Attenuation_0], properties {bLoopIndefinitely=false}}
 {id=Mixer_0, children [Attenuation_1, Looping_0], properties {inputVolume [0.8,1.2]}}
 {
  id=WavePlayer_4
  properties {soundWaveAssetPtr=/Game/Sounds/Helios/Environment_Actors/NitroBossVehicleGarage/Nitro_BossVehicleGarage_Gate_Open_01}
 }
]
properties {
 bOverrideConcurrency=true
 maxDistance=26614.2852
 soundClassObject=/Game/Sounds/FortSoundClassesAndModes/Fort_Enemy_Attack
 volumeMultiplier=1.5
 attenuationOverrides {
  bEnableListenerFocus=true
  distanceAlgorithm=Logarithmic
  falloffDistance=1200
  focusDistanceScale=0.9
  nonFocusVolumeAttenuation=0.4
  stereoSpread=700
  attenuationShapeExtents=<250 120 -60.9851>
  customAttenuationCurve {editorCurveData {keys [{arriveTangent=-2.3812, interpMode=RCIM_Cubic, leaveTangent=-2.3812, tangentMode=RCTM_User, time=0, value=1}, <at 1 0.01 RCIM_Cubic>]}}
 }
}
