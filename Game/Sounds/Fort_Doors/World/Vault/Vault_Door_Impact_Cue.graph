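# Sound cue for metallic resonant impacts on a vault door, routed to the Fort_Impacts sound class.
# A Random node picks one of four equally weighted resonate wave variants per play, at slightly reduced volume.
# It uses the small-object demolition hit attenuation and a four-max object impact concurrency set, alongside override values that stop the oldest voice.
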SoundCue /Game/Sounds/Fort_Doors/World/Vault/Vault_Door_Impact_Cue
root=Random_0
nodes [
 {id=Random_0, children [WavePlayer_5, WavePlayer_9, WavePlayer_10, WavePlayer_11], properties {weights [1,1,1,1]}}
 {id=WavePlayer_10, properties {soundWaveAssetPtr=&/Vault_Door_Impact_Mtl_Resonate_03}}
 {id=WavePlayer_11, properties {soundWaveAssetPtr=&/Vault_Door_Impact_Mtl_Resonate_04}}
 {id=WavePlayer_5, properties {soundWaveAssetPtr=&/Vault_Door_Impact_Mtl_Resonate_01}}
 {id=WavePlayer_9, properties {soundWaveAssetPtr=&/Vault_Door_Impact_Mtl_Resonate_02}}
]
properties {
 attenuationSettings=/Game/Sounds/Attenuation/Fort_Demolition_Hit_Small_Sized_Object_Att
 maxDistance=4000
 soundClassObject=&Fort_Impacts
 volumeMultiplier=0.8
 concurrencyOverrides {maxCount=3, resolutionRule=StopOldest}
 concurrencySet [&CONC_Object_Impacts_FourMax]
}
