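# Sound cue for the vault door keycard ping, built from three reused PropHunt nearby-ping waves.
# A Mixer layers the waves at different volumes behind a random start delay of up to a second.
# Attenuation and concurrency overrides switch off distance volume falloff and cap simultaneous instances.
# It plays at reduced overall volume in the critical in-game UI sound class.
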SoundCue /Game/Sounds/Fort_Doors/World/Vault/Vault_Door_Keycard_Ping_Cue
root=Delay_0
nodes [
 {id=Delay_0, children [Mixer_2], properties {delayMax=1}}
 {id=Mixer_2, children [WavePlayer_8,WavePlayer_6,WavePlayer_7], properties {inputVolume [2,0.7,1]}}
 {id=WavePlayer_6, properties {soundWaveAssetPtr=/Game/Sounds/Creative/Modes/PropHunt/PropHunt_Nearby_Ping_03}}
 {id=WavePlayer_7, properties {soundWaveAssetPtr=/Game/Sounds/Creative/Modes/PropHunt/PropHunt_Nearby_Ping_01}}
 {id=WavePlayer_8, properties {soundWaveAssetPtr=/Game/Sounds/Creative/Modes/PropHunt/PropHunt_Nearby_Ping_05}}
]
properties {
 attenuationSettings=/Game/Sounds/Creative/Attenuations/Fort_Creative_PropHunt_Ping_Att
 bOverrideAttenuation=true
 bOverrideConcurrency=true
 maxDistance=2097152
 soundClassObject=&Fort_UI_InGame_Critical
 volumeMultiplier=0.35
 attenuationOverrides {bAttenuate=false}
 concurrencyOverrides {maxCount=3}
}
