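# Sound cue for the slam of the Moonflax vault gate opening, judging by its name.
# A single wave player, using a gate-open slam wave from the Helios Nitro boss vehicle garage, feeds an Attenuation node that overrides falloff with a logarithmic curve.
# The cue overrides concurrency, boosts volume slightly and routes to the Fort_Enemy_Attack sound class rather than a door-specific one.
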SoundCue /Game/Sounds/Fort_Doors/World/Vault/Vault_Door_Moonflax_Gate_Open_Slam_Cue
root=Attenuation_1
nodes [
 {
  id=Attenuation_1
  children [WavePlayer_0]
  properties {bOverrideAttenuation=true, attenuationOverrides {distanceAlgorithm=Logarithmic, falloffDistance=4500, attenuationShapeExtents {x=700}}}
 }
 {
  id=WavePlayer_0
  properties {soundWaveAssetPtr=/Game/Sounds/Helios/Environment_Actors/NitroBossVehicleGarage/Nitro_BossVehicleGarage_Gate_Open_Slam_01}
 }
]
properties {
 bOverrideConcurrency=true
 maxDistance=5200
 soundClassObject=/Game/Sounds/FortSoundClassesAndModes/Fort_Enemy_Attack
 volumeMultiplier=1.2
 attenuationOverrides {
  bEnableListenerFocus=true
  distanceAlgorithm=Logarithmic
  falloffDistance=1200
  focusDistanceScale=0.9
  nonFocusVolumeAttenuation=0.4
  stereoSpread=700
  attenuationShapeExtents=<250 120 -60.9851>
  customAttenuationCurve {editorCurveData {keys [{arriveTangent=-2.3812, interpMode=RCIM_Cubic, leaveTangent=-2.3812, tangentMode=RCTM_User, time=0, value=1}, <at 1 0.01 RCIM_Cubic>]}}
 }
}
